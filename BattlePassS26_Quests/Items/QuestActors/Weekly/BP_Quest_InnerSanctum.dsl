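; Weekly quest actor blueprint for the Season 26 Inner Sanctum quest, adding editor-visible material swaps to its parent quest actor.
; The construction script applies optional pre- or post-interaction material overrides to the extra meshes, driven by the post-interaction preview flag.
; The post override falls back to the pre-interaction mesh when no post mesh exists; without an override, that mesh receives its cached original material.
; PostInteractionVisuals only calls the parent.
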(fn PostInteractionVisuals
  (params (PlayerController :type Fort%20Player%20Controller%20Object%20Reference))
  (|CallParentFunction PlayerController))

(fn ConstructionScript
  (bind _preinteraction_mis_override (Variables|Design|PreInteraction|GetPreInteraction_MIS_Override))
  (bind _postinteraction_mis_override (Variables|Design|PostInteraction|GetPostInteraction_MIS_Override))
  (|CallParentFunction)
  (Utilities|FlowControl|Sequence
    (:then_1
      ;[1 1 1] Material Swaps
      (Utilities|IsValid _preinteraction_mis_override
        (:Is%20Valid
          (LocalVariables|SetStoredOriginalPreInteractionExtraMaterial (Utilities|Array|Get(acopy) (Rendering|Material|GetMaterials (Variables|Default|GetSM_Quest_PreInteraction_Extra))))
          (if (not (Variables|Design|PostInteraction|GetbPreviewPostInteraction))
            ;;[1 1 1] If PreInteraction has swapped material
            (Rendering|Material|SetMaterial (Variables|Default|GetSM_Quest_PreInteraction_Extra) 0 _preinteraction_mis_override)))
        (:Is%20Not%20Valid
          (if (Variables|Design|PostInteraction|GetbPreviewPostInteraction)
            _)))
      ;;[1 1 1] If Previewing the Post Interaction State
      (Utilities|IsValid _postinteraction_mis_override
        (:Is%20Valid
          (Utilities|IsValid (Variables|Default|GetSM_Quest_PostInteraction_Extra)
            (:Is%20Valid
              (Rendering|Material|SetMaterial (Variables|Default|GetSM_Quest_PostInteraction_Extra) 0 _postinteraction_mis_override))
            (:Is%20Not%20Valid
              (Rendering|Material|SetMaterial (Variables|Default|GetSM_Quest_PreInteraction_Extra) 0 _postinteraction_mis_override))))
        (:Is%20Not%20Valid
          (Rendering|Material|SetMaterial (Variables|Default|GetSM_Quest_PreInteraction_Extra) 0 (LocalVariables|GetStoredOriginalPreInteractionExtraMaterial)))))))
